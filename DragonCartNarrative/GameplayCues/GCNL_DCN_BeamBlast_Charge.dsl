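; Looping gameplay cue for the flying orb's beam blast charge, driving targeting and charge-up VFX, audio and sound indicator.
; Loop start finds the orb, feeds its mesh and targeting time to Niagara, enables its aggro state and starts targeting audio.
; A timer at 75% of targeting time plays charge audio on the target.
; Removal positions effects using target chest and orb final hit, times charge-up by laser fire time, and ends targeting feedback.
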(event EventBeginPlay
  (|CallParentFunction)
  (Development|PrintString 0 "BEAM VFX ON"))

(event EventOnLoopingStart(Niagara)
  (params (MyTarget :type Actor%20Object%20Reference) (Parameters :type Gameplay%20Cue%20Parameters%20Structure%20%28by%20ref%29) (ParticleComponents :type Array%20of%20Niagara%20Particle%20System%20Component%20Object%20References) (AudioComponents :type Array%20of%20Audio%20Component%20Object%20References))
  (bind _output (Utilities|Array|Get(acopy) ParticleComponents))
  (bind _orbbp (Variables|Default|GetOrbBP))
  (bind _staticmesh (Variables|/DragonCartNarrative/Gameplay/FlyingOrb/BGA_DCN_OrbActor.BGA_DCN_OrbActor_C|GetStaticMesh _orbbp))
  (Variables|Default|SetTarget MyTarget)
  (Variables|Default|GetOrbBP
    (:then
      _)
    (:else
      (bind _returnvalue (Actor|GetActorOfClass 0 "/DragonCartNarrative/Gameplay/FlyingOrb/BGA_DCN_OrbActor.BGA_DCN_OrbActor_C"))
      (Variables|Default|SetOrbBP _returnvalue)))
  (Niagara|SetNiagaraStaticMeshComponent _output "Static Mesh Reader" _staticmesh)
  (Niagara|SetNiagaraVariable(Position) _output "BeamPosition" (Transformation|GetWorldLocation _staticmesh))
  (Effects|Components|ParticleSystem|SetFloatParameter _output "TargetingDuration" (ScalableFloat|GetValueatLevel (Variables|Default|GetOrb_TargetingTime)))
  (Effects|Components|ParticleSystem|SetBoolParameter (Variables|/DragonCartNarrative/Gameplay/FlyingOrb/BGA_DCN_OrbActor.BGA_DCN_OrbActor_C|GetNiagara%20-%20aggro%20state (Variables|Default|GetOrbBP)) "Targeting" true)
  (SoundIndicator|StartTargetedSoundIndicatorLooping (Variables|Default|GetFortSoundIndicator) (Variables|Default|GetOrbBP) (Variables|Default|GetIn%20Sound%20Being%20Played) 1.0 0)
  (bind _orbbp_2 (Variables|Default|GetOrbBP)
    (:then
      (Variables|Default|GetOrbTargetAudio
        (:then)
        (:else
          (bind _returnvalue_2 (Audio|SpawnSoundAttached "/DragonCartNarrative/Assets/Audio/Laser/MSS_DragonCart_Orb_Targeting.MSS_DragonCart_Orb_Targeting" (Variables|/DragonCartNarrative/Gameplay/FlyingOrb/BGA_DCN_OrbActor.BGA_DCN_OrbActor_C|GetStaticMesh _orbbp_2)))
          (Variables|Default|SetOrbTargetAudio _returnvalue_2)
          (Utilities|Time|SetTimerbyEvent (EventDispatchers|CreateEvent :!event Custom|TargetingLocked) (* (ScalableFloat|GetValueatLevel (Variables|Default|GetOrb_TargetingTime)) 0.75)))))
    (:else)))

(event Custom|TargetingLocked
  (bind _target (Variables|Default|GetTarget)
    (:then
      (bind _returnvalue (Audio|SpawnSoundAttached "/DragonCartNarrative/Assets/Audio/Laser/MSS_DragonCart_Laser_ChargeUp_Target.MSS_DragonCart_Laser_ChargeUp_Target" (Variables|/Script/Engine.Actor|GetRootComponent _target)))
      (Variables|Default|SetTargetChargeAudio _returnvalue))
    (:else)))

(event EventOnRemoval(Niagara)
  (params (MyTarget :type Actor%20Object%20Reference) (Parameters :type Gameplay%20Cue%20Parameters%20Structure%20%28by%20ref%29) (ParticleComponents :type Array%20of%20Niagara%20Particle%20System%20Component%20Object%20References) (AudioComponents :type Array%20of%20Audio%20Component%20Object%20References) (BurstCameraShakeInstance :type Camera%20Shake%20Base%20Object%20Reference) (BurstDecalInstance :type Decal%20Actor%20Object%20Reference))
  (bind _laser_firetime (Variables|Default|GetLaser_FireTime))
  (bind _returnvalue (ScalableFloat|GetValueatLevel _laser_firetime))
  (bind _output (Utilities|Array|Get(acopy) ParticleComponents))
  (SoundIndicator|StopSoundIndicatorLooping (Variables|Default|GetFortSoundIndicator))
  (bind _asfort_pawn (Utilities|Casting|CastToFortPawn MyTarget)
    (:then
      (Niagara|SetNiagaraVariable(Position) _output "InitialPosition" (Transformation|GetSocketLocation (Variables|/Script/Engine.Character|GetMesh _asfort_pawn) "FX_Chest")))
    (:CastFailed
      _))
  (bind _orbbp (Variables|Default|GetOrbBP)
    (:then
      (Transformation|SetWorldLocation _output (Variables|/DragonCartNarrative/Gameplay/FlyingOrb/BGA_DCN_OrbActor.BGA_DCN_OrbActor_C|GetFinalHit _orbbp) false false)
      (Niagara|SetNiagaraStaticMeshComponent _output "Static Mesh Reader" (Variables|/DragonCartNarrative/Gameplay/FlyingOrb/BGA_DCN_OrbActor.BGA_DCN_OrbActor_C|GetStaticMesh _orbbp))
      (Effects|Components|ParticleSystem|SetFloatParameter _output "Duration" _returnvalue)
      (bind _returnvalue_2 (Audio|SpawnSoundAttached "/DragonCartNarrative/Assets/Audio/Laser/MSS_DragonCart_Laser_ChargeUp_Orb.MSS_DragonCart_Laser_ChargeUp_Orb" (Variables|/DragonCartNarrative/Gameplay/FlyingOrb/BGA_DCN_OrbActor.BGA_DCN_OrbActor_C|GetStaticMesh _orbbp)))
      (Audio|Parameter|SetFloatParameter _returnvalue_2 "ChargeDuration" _returnvalue :!class /Script/AudioExtensions.AudioParameterControllerInterface)
      (Effects|Components|ParticleSystem|SetBoolParameter (Variables|/DragonCartNarrative/Gameplay/FlyingOrb/BGA_DCN_OrbActor.BGA_DCN_OrbActor_C|GetNiagara%20-%20aggro%20state (Variables|Default|GetOrbBP)) "Targeting" false)
      (bind _targetchargeaudio (Variables|Default|GetTargetChargeAudio)
        (:then
          (Audio|Parameter|SetFloatParameter _targetchargeaudio "ChargeDuration" _returnvalue :!class /Script/AudioExtensions.AudioParameterControllerInterface))
        (:else
          _))
      (bind _orbtargetaudio (Variables|Default|GetOrbTargetAudio)
        (:then
          (Audio|Components|Audio|FadeOut _orbtargetaudio 0.5 0.0 "Linear" :!class /Script/Engine.AudioComponent))
        (:else)))
    (:else)))

(fn ConstructionScript)
