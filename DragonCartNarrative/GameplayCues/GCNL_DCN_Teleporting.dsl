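; Looping gameplay cue notify for DragonCartNarrative teleporting, giving the locally controlled player a Zero Point camera lens effect and first-person travel audio.
; On start it destroys any leftover lens effect, adds a Niagara one and restarts the 2D sound.
; On removal a timeline drives the lens FlashOut, then removes it with a portal-warp camera shake.
; For the owning pawn, removal also delays rift particles and fades the audio.
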(event EventOnLoopingStart(EitherVFX-preferthisversion)
  (params (MyTarget :type Actor%20Object%20Reference) (Parameters :type Gameplay%20Cue%20Parameters%20Structure%20%28by%20ref%29) (ParticleComponents :type Array%20of%20FXSystem%20Component%20Object%20References) (AudioComponents :type Array%20of%20Audio%20Component%20Object%20References))
  (bind _lens_effect_niagara (Variables|Default|GetLens%20Effect%20Niagara)
    (:then
      (Actor|DestroyActor _lens_effect_niagara))
    (:else
      _))
  (bind _asfort_player_pawn (Utilities|Casting|CastToFortPlayerPawn MyTarget))
  (if (Pawn|IsLocallyControlled _asfort_player_pawn)
    (Variables|Default|SetOwningActor _asfort_player_pawn)
    (bind _returnvalue (Camera|AddGenericCameraLensEffect (Variables|/Script/Engine.PlayerController|GetPlayerCameraManager (Fortnite|GetFortPlayerControllerfromActor _asfort_player_pawn)) "/Game/Effects/Fort_Effects/Camera_Effects/BP_CameraLens_HidingProp_ZeroPointTeleporting_NS.BP_CameraLens_HidingProp_ZeroPointTeleporting_NS_C"))
    (bind _output_get (Variables|Default|SetLensEffectInterface _returnvalue))
    (bind _asniagara_lens_effect_base (Utilities|Casting|CastToNiagaraLensEffectBase _output_get))
    (Variables|Default|SetLens%20Effect%20Niagara _asniagara_lens_effect_base)
    (Utilities|FlowControl|Sequence
      (:then_0
        (bind _1ptravelaudio (Variables|Default|Get1PTravelAudio)
          (:then
            (Audio|Components|Audio|FadeOut _1ptravelaudio 0.25 0.0 "Linear" :!class /Script/Engine.AudioComponent))
          (:else)))
      (:then_1
        (bind _returnvalue_2 (Audio|SpawnSound2D 0 "/DragonCartNarrative/Assets/Audio/Shard/ZP_DragonCart_Shard_Travel_1P_Cue.ZP_DragonCart_Shard_Travel_1P_Cue"))
        (Variables|Default|Set1PTravelAudio _returnvalue_2)))))

(event EventOnRemoval(EitherVFX-preferthisversion)
  (params (MyTarget :type Actor%20Object%20Reference) (Parameters :type Gameplay%20Cue%20Parameters%20Structure%20%28by%20ref%29) (ParticleComponents :type Array%20of%20FXSystem%20Component%20Object%20References) (AudioComponents :type Array%20of%20Audio%20Component%20Object%20References) (BurstCameraShakeInstance :type Camera%20Shake%20Base%20Object%20Reference) (BurstDecalInstance :type Decal%20Actor%20Object%20Reference))
  (bind _asfort_player_pawn (Utilities|Casting|CastToFortPlayerPawn MyTarget))
  (Utilities|FlowControl|Sequence
    (:then_0
      (bind _lens_effect_niagara (Variables|Default|GetLens%20Effect%20Niagara)
        (:then
          (bind (_direction _update_white_wash) (Timeline|Timeline PlayFromStart)
            (:Update
              (Effects|Components|ParticleSystem|SetFloatParameter (Variables|/Script/Niagara.NiagaraActor|GetNiagaraComponent _lens_effect_niagara) "FlashOut" _update_white_wash))
            (:Finished
              (Components|Activation|Deactivate (Variables|/Script/Niagara.NiagaraActor|GetNiagaraComponent _lens_effect_niagara))
              (Camera|RemoveGenericCameraLensEffect (Variables|/Script/Engine.PlayerController|GetPlayerCameraManager (Fortnite|GetFortPlayerControllerfromActor _asfort_player_pawn)) (Variables|Default|GetLensEffectInterface))
              (Game|Feedback|ClientStartCameraShake :self (Fortnite|GetFortPlayerControllerfromActor _asfort_player_pawn) :Shake "/FNGameplayCues/Effects/Fort_Effects/Camera_Effects/BP_CameraShake_PortalWarp_Fast.BP_CameraShake_PortalWarp_Fast_C" :Scale 1.0 :PlaySpace "CameraLocal")
              (Actor|DestroyActor (Variables|Default|GetLens%20Effect%20Niagara)))))
        (:else)))
    (:then_1
      (if (Utilities|Equal(Object) (Variables|Default|GetOwningActor) _asfort_player_pawn)
        (Effects|Components|ParticleSystem|SetFloatParameter (Utilities|Array|Get(acopy) ParticleComponents) "StartDelay" 0.25) ; Rift FX delay until the tunnel is finished for local player
        (bind _1ptravelaudio (Variables|Default|Get1PTravelAudio)
          (:then
            (Audio|Components|Audio|FadeOut _1ptravelaudio 1.0 0.0 "Linear" :!class /Script/Engine.AudioComponent))
          (:else))))))

(fn ConstructionScript)
