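; Gameplay cue notify that gives audio feedback when a pawn is marked by an opponent.
; On burst, it calls the parent handler, then casts the target to a Fort player pawn.
; If that pawn is locally viewed, it spawns the MarkedByOpponent sound attached to the pawn's root component, so only clients viewing that pawn hear it.
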(fn OnBurst(EitherVFX-preferthisversion)
  (params (MyTarget :type Actor%20Object%20Reference) (Parameters :type Gameplay%20Cue%20Parameters%20Structure%20%28by%20ref%29) (ParticleComponents :type Array%20of%20FXSystem%20Component%20Object%20References) (AudioComponents :type Array%20of%20Audio%20Component%20Object%20References) (BurstCameraShakeInstance :type Camera%20Shake%20Base%20Object%20Reference) (BurstDecalInstance :type Decal%20Actor%20Object%20Reference))
  ;[0.15 0.04667 0.148 0.5] When a pawn is marked by an opponent, play a sound for any clients viewing the pawn
  (|CallParentFunction MyTarget Parameters ParticleComponents AudioComponents BurstCameraShakeInstance BurstDecalInstance)
  (bind _asfort_player_pawn (Utilities|Casting|CastToFortPlayerPawn MyTarget))
  (if (Pawn|IsLocallyViewed _asfort_player_pawn)
    (Audio|SpawnSoundAttached (Variables|Default|GetMarkedByOpponentSound) (Variables|/Script/Engine.Actor|GetRootComponent _asfort_player_pawn))
    (else
      _))
  (return))
