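; Gameplay cue that plays the flying orb's death-laser blast effects.
; On execute it finds the orb via GetActorOfClass, attaches a Niagara beam at its EyeSocket, and feeds cue location as endpoint and raw magnitude as DistanceDuration.
; For pawn instigators, fire audio faces the endpoint with 1P/3P attenuation by local view, whose falloff also sizes a visual sound indicator.
; An OnBurst handler also sets the endpoint on a burst particle component.
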(fn OnBurst(EitherVFX-preferthisversion)
  (params (MyTarget :type Actor%20Object%20Reference) (Parameters :type Gameplay%20Cue%20Parameters%20Structure%20%28by%20ref%29) (ParticleComponents :type Array%20of%20FXSystem%20Component%20Object%20References) (AudioComponents :type Array%20of%20Audio%20Component%20Object%20References) (BurstCameraShakeInstance :type Camera%20Shake%20Base%20Object%20Reference) (BurstDecalInstance :type Decal%20Actor%20Object%20Reference))
  (bind (_normalizedmagnitude _rawmagnitude _effectcontext _matchedtagname _originaltag _aggregatedsourcetags _aggregatedtargettags _location) (Ability|GameplayCue|BreakGameplayCueParameters Parameters))
  (Effects|Components|ParticleSystem|SetVectorParameter (Utilities|Array|Get(acopy) ParticleComponents) "EndPoint" _location)
  (return))

(fn OnExecute
  (params (MyTarget :type Actor%20Object%20Reference) (Parameters :type Gameplay%20Cue%20Parameters%20Structure%20%28by%20ref%29))
  (bind (_normalizedmagnitude _rawmagnitude _effectcontext _matchedtagname _originaltag _aggregatedsourcetags _aggregatedtargettags _location _normal _instigator) (Ability|GameplayCue|BreakGameplayCueParameters Parameters))
  (bind _returnvalue (Actor|GetActorOfClass 0 "/DragonCartNarrative/Gameplay/FlyingOrb/BGA_DCN_OrbActor.BGA_DCN_OrbActor_C"))
  (bind _returnvalue_2 (Niagara|SpawnSystemAttached "/DragonCartNarrative/Assets/VFX/Orb_Attack/NS_DragonCartNarrative_Orb_DeathLaser.NS_DragonCartNarrative_Orb_DeathLaser" (Variables|/DragonCartNarrative/Gameplay/FlyingOrb/BGA_DCN_OrbActor.BGA_DCN_OrbActor_C|GetStaticMesh _returnvalue) "EyeSocket"))
  (Effects|Components|ParticleSystem|SetFloatParameter _returnvalue_2 "DistanceDuration" _rawmagnitude)
  (Effects|Components|ParticleSystem|SetVectorParameter _returnvalue_2 "EndPoint" _location)
  (bind _asfort_pawn (Utilities|Casting|CastToFortPawn _instigator)
    (:then
      (Audio|SpawnSoundAttached :Sound "/DragonCartNarrative/Assets/Audio/Laser/MSS_DragonCart_Orb_Fire.MSS_DragonCart_Orb_Fire" :AttachToComponent (Variables|/DragonCartNarrative/Gameplay/FlyingOrb/BGA_DCN_OrbActor.BGA_DCN_OrbActor_C|GetStaticMesh _returnvalue) :AttachPointName "EyeSocket" :Rotation (Math|Rotator|FindRelativeLookatRotation (Transformation|GetSocketTransform (Variables|/DragonCartNarrative/Gameplay/FlyingOrb/BGA_DCN_OrbActor.BGA_DCN_OrbActor_C|GetStaticMesh _returnvalue) "EyeSocket" "RTS_World") _location) :LocationType "KeepRelativeOffset" :bStopWhenAttachedToDestroyed false :VolumeMultiplier 1.0 :PitchMultiplier 1.0 :StartTime 0.0 :AttenuationSettings (select (Pawn|IsLocallyViewed _asfort_pawn) "/DragonCartNarrative/Assets/Audio/Laser/ATT_DragonCart_Orb_Laser_Fire_1P.ATT_DragonCart_Orb_Laser_Fire_1P" "/DragonCartNarrative/Assets/Audio/Laser/ATT_DragonCart_Orb_Laser_Fire_3P.ATT_DragonCart_Orb_Laser_Fire_3P") :bAutoDestroy true))
    (:CastFailed
      _))
  (Fortnite|ShowSoundIndicatorLocal :WorldContextObject MyTarget :Location (Transformation|GetActorLocation _returnvalue) :MaxAudibleDistance (.FalloffDistance (Utilities|Struct|BreakSoundAttenuationSettings (Variables|/Script/Engine.SoundAttenuation|GetAttenuation (select (Pawn|IsLocallyViewed _asfort_pawn) "/DragonCartNarrative/Assets/Audio/Laser/ATT_DragonCart_Orb_Laser_Fire_1P.ATT_DragonCart_Orb_Laser_Fire_1P" "/DragonCartNarrative/Assets/Audio/Laser/ATT_DragonCart_Orb_Laser_Fire_3P.ATT_DragonCart_Orb_Laser_Fire_3P")))) :IndicatorType "Gunshot" :Instigator MyTarget :Tint "(R=0.000000,G=0.000000,B=0.000000,A=0.000000)" :OverrideIcon "/DragonCartNarrative/Gameplay/Icons/T_UI_DragonCartNarrative_AV.T_UI_DragonCartNarrative_AV" :SoundBase (Variables|Default|GetSoundIndicatorSound) :AttenuationMaxRangeMultiplier 1.0 :AttenuationIntensityMultiplier 1.0)
  (return (|CallParentFunction MyTarget Parameters)))
