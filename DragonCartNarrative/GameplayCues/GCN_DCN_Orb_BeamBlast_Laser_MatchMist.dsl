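; Gameplay cue that plays the Match Mist orb turret's laser beam blast.
; On execute, it attaches a Niagara death laser to the turret's FirePoint, feeding raw magnitude as DistanceDuration and cue location as EndPoint.
; If the instigator is a Fort pawn, a fire sound plays from the eye socket with first- or third-person attenuation, depending on local viewing.
; A burst handler also sets EndPoint on the cue's particle components.
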(fn OnBurst(EitherVFX-preferthisversion)
  (params (MyTarget :type Actor%20Object%20Reference) (Parameters :type Gameplay%20Cue%20Parameters%20Structure%20%28by%20ref%29) (ParticleComponents :type Array%20of%20FXSystem%20Component%20Object%20References) (AudioComponents :type Array%20of%20Audio%20Component%20Object%20References) (BurstCameraShakeInstance :type Camera%20Shake%20Base%20Object%20Reference) (BurstDecalInstance :type Decal%20Actor%20Object%20Reference))
  (bind (_normalizedmagnitude _rawmagnitude _effectcontext _matchedtagname _originaltag _aggregatedsourcetags _aggregatedtargettags _location) (Ability|GameplayCue|BreakGameplayCueParameters Parameters))
  (Effects|Components|ParticleSystem|SetVectorParameter (Utilities|Array|Get(acopy) ParticleComponents) "EndPoint" _location)
  (return))

(fn OnExecute
  (params (MyTarget :type Actor%20Object%20Reference) (Parameters :type Gameplay%20Cue%20Parameters%20Structure%20%28by%20ref%29))
  (bind (_normalizedmagnitude _rawmagnitude _effectcontext _matchedtagname _originaltag _aggregatedsourcetags _aggregatedtargettags _location _normal _instigator) (Ability|GameplayCue|BreakGameplayCueParameters Parameters))
  (bind _asbp_dcn_turret_match_mist (Utilities|Casting|CastToBP_DCN_Turret_MatchMist (Ability|GameplayCue|GetInstigatorActor Parameters)))
  (bind _returnvalue (Niagara|SpawnSystemAttached "/DragonCartNarrative/Assets/VFX/Orb_Attack/NS_DragonCartNarrative_Orb_DeathLaser.NS_DragonCartNarrative_Orb_DeathLaser" (Variables|/DragonCartNarrative/MatchMist/BP_DCN_Turret_MatchMist.BP_DCN_Turret_MatchMist_C|GetTurret _asbp_dcn_turret_match_mist) "FirePoint"))
  (Effects|Components|ParticleSystem|SetFloatParameter _returnvalue "DistanceDuration" _rawmagnitude)
  (Effects|Components|ParticleSystem|SetVectorParameter _returnvalue "EndPoint" _location)
  (bind _asfort_pawn (Utilities|Casting|CastToFortPawn _instigator)
    (:then
      (Audio|SpawnSoundAttached :Sound "/DragonCartNarrative/Assets/Audio/Laser/MSS_DragonCart_Orb_Fire.MSS_DragonCart_Orb_Fire" :AttachToComponent (Variables|/DragonCartNarrative/MatchMist/BP_DCN_Turret_MatchMist.BP_DCN_Turret_MatchMist_C|GetTurret _asbp_dcn_turret_match_mist) :AttachPointName "EyeSocket" :Rotation (Math|Rotator|FindRelativeLookatRotation (Transformation|GetSocketTransform (Variables|/DragonCartNarrative/MatchMist/BP_DCN_Turret_MatchMist.BP_DCN_Turret_MatchMist_C|GetTurret _asbp_dcn_turret_match_mist) "FirePoint" "RTS_World") _location) :LocationType "KeepRelativeOffset" :bStopWhenAttachedToDestroyed false :VolumeMultiplier 1.0 :PitchMultiplier 1.0 :StartTime 0.0 :AttenuationSettings (select (Pawn|IsLocallyViewed _asfort_pawn) "/DragonCartNarrative/Assets/Audio/Laser/ATT_DragonCart_Orb_Laser_Fire_1P.ATT_DragonCart_Orb_Laser_Fire_1P" "/DragonCartNarrative/Assets/Audio/Laser/ATT_DragonCart_Orb_Laser_Fire_3P.ATT_DragonCart_Orb_Laser_Fire_3P") :bAutoDestroy true))
    (:CastFailed
      _))
  (return (|CallParentFunction MyTarget Parameters)))
