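; Gameplay cue notify for the DragonCart flying orb's reveal: a timed world-sweep pulse with sound.
; OnExecute stores cue magnitude as radius and location as origin, attaches a Niagara pulse at the orb's EyeSocket, triggers its ripple, and pushes a sound mix.
; A delayed timer enables post-process and feeds the sweep to FringePlank's Interrogate material parameter collection.
; A second timer disables post-process; EndPlay clears both timers and does likewise.
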(event Custom|BeginSweepFX
  (bind _targetorb (Variables|Default|GetTargetOrb))
  (bind _staticmesh (Variables|/DragonCartNarrative/Gameplay/FlyingOrb/BGA_DCN_OrbActor.BGA_DCN_OrbActor_C|GetStaticMesh _targetorb))
  (bind (_returnvalue_location _returnvalue_rotation) (Math|Transform|BreakTransform (Transformation|GetSocketTransform _staticmesh "EyeSocket" "RTS_World")))
  (bind _returnvalue (Utilities|Time|SetTimerbyEvent (EventDispatchers|CreateEvent :!event Custom|EndSweepFX) (+ (ScalableFloat|GetValueatLevel (Variables|Default|GetStencilDuration)) (Variables|Default|GetDelayBeforeSweep))))
  (Variables|Default|SetSweepDelayHandle _returnvalue)
  (bind _returnvalue_2 (Utilities|Time|SetTimerbyEvent (EventDispatchers|CreateEvent :!event Custom|PlayWorldSweep) (Variables|Default|GetDelayBeforeSweep)))
  (Variables|Default|SetDelayBeforeSweepHandle _returnvalue_2)
  (if (Utilities|IsValid _targetorb :!class /Script/Engine.KismetSystemLibrary)
    (bind _returnvalue_3 (Niagara|SpawnSystemAttached "/DragonCartNarrative/Assets/VFX/Pulse/NS_DragonCart_Orb_Pulse.NS_DragonCart_Orb_Pulse" _staticmesh "EyeSocket" "0, 0, 0" "0, 0, 0" "SnapToTarget"))
    (Effects|Components|ParticleSystem|SetFloatParameter _returnvalue_3 "PulseRadius" (Variables|Default|GetSweepRadius))
    (Effects|Components|ParticleSystem|SetFloatParameter _returnvalue_3 "Duration" (Variables|Default|GetSweepDuration))
    (Class|BGADCNOrbActor|TriggerRipple (Variables|Default|GetTargetOrb) (+ _returnvalue_location (* (Math|Rotator|GetRotationXVector _returnvalue_rotation) 5e+02)) 1.5e+04 4.0)))

(event Custom|EndSweepFX
  (Variables|/Script/Engine.PostProcessComponent|SetBlendWeight 0.0 (Variables|Default|GetPostProcess))
  (Variables|/Script/Engine.PostProcessComponent|SetbEnabled false (Variables|Default|GetPostProcess)))

(event Custom|PlayWorldSweep
  (Variables|/Script/Engine.PostProcessComponent|SetBlendWeight 1.0 (Variables|Default|GetPostProcess))
  (Variables|/Script/Engine.PostProcessComponent|SetbEnabled true (Variables|Default|GetPostProcess))
  (Rendering|Material|SetVectorParameterValue 0 "/Game/Weapons/Prototype/FringePlank/PostProcess/MPC_FringePlank_Interrogate.MPC_FringePlank_Interrogate" "OriginPositionWS" (Math|Conversions|ToLinearColor(Vector) (Variables|Default|GetLocation)) :!class /Script/Engine.KismetMaterialLibrary)
  (Rendering|Material|SetScalarParameterValue 0 "/Game/Weapons/Prototype/FringePlank/PostProcess/MPC_FringePlank_Interrogate.MPC_FringePlank_Interrogate" "Duration" (Variables|Default|GetSweepDuration) :!class /Script/Engine.KismetMaterialLibrary)
  (Rendering|Material|SetScalarParameterValue 0 "/Game/Weapons/Prototype/FringePlank/PostProcess/MPC_FringePlank_Interrogate.MPC_FringePlank_Interrogate" "StencilDuration" (ScalableFloat|GetValueatLevel (Variables|Default|GetStencilDuration)) :!class /Script/Engine.KismetMaterialLibrary)
  (Rendering|Material|SetScalarParameterValue 0 "/Game/Weapons/Prototype/FringePlank/PostProcess/MPC_FringePlank_Interrogate.MPC_FringePlank_Interrogate" "Radius" (Variables|Default|GetSweepRadius) :!class /Script/Engine.KismetMaterialLibrary)
  (Rendering|Material|SetScalarParameterValue 0 "/Game/Weapons/Prototype/FringePlank/PostProcess/MPC_FringePlank_Interrogate.MPC_FringePlank_Interrogate" "ActivatedTimestamp" (Utilities|Time|GetGameTimeinSeconds 0) :!class /Script/Engine.KismetMaterialLibrary))

(event EventEndPlay
  (params (EndPlayReason :type EEndPlayReason%20Enum :default "Destroyed"))
  (Utilities|Time|ClearandInvalidateTimerbyHandle 0 (Variables|Default|GetSweepDelayHandle))
  (Utilities|Time|ClearandInvalidateTimerbyHandle 0 (Variables|Default|GetDelayBeforeSweepHandle))
  (CallFunction|EndSweepFX))

(fn OnExecute
  (params (MyTarget :type Actor%20Object%20Reference) (Parameters :type Gameplay%20Cue%20Parameters%20Structure%20%28by%20ref%29))
  (bind (_normalizedmagnitude _rawmagnitude _effectcontext _matchedtagname _originaltag _aggregatedsourcetags _aggregatedtargettags _location) (Ability|GameplayCue|BreakGameplayCueParameters Parameters))
  (|CallParentFunction MyTarget Parameters)
  (Utilities|FlowControl|Sequence
    (:then_0
      ; Log
      (Fortnite|FortShippingLog self (Utilities|String|Append "[GCN_DCN_Orb_Reveal] - OnExecute - Radius=" (Utilities|String|ToString(Float) _rawmagnitude) "  Location=" (Utilities|String|ToString(Vector) _location))))
    (:then_1
      ; Store info for later
      (Variables|Default|SetSweepRadius _rawmagnitude)
      (Variables|Default|SetLocation _location)
      (bind _asbga_dcn_orb_actor (Utilities|Casting|CastToBGA_DCN_OrbActor MyTarget)
        (:then
          (Variables|Default|SetTargetOrb _asbga_dcn_orb_actor))
        (:CastFailed
          (return false))))
    (:then_2
      ; VFX
      (CallFunction|BeginSweepFX))
    (:then_3
      ; SFX
      (Audio|SpawnSoundatLocation 0 (Variables|Default|GetSweep%20Sound) (Variables|Default|GetLocation))
      (Audio|PushSoundMixModifier 0 (Variables|Default|GetRevealSoundMix)))
    (:then_4
      (return true))))

(fn ConstructionScript)
